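; Base interactable button actor for BasilOkra that shows its static mesh only while a player has interaction focus on it.
; The mesh is hidden at BeginPlay and toggled on focus and unfocus, only where cosmetic events can execute.
; Interaction is gated by bCanInteract, and the prompt is placeholder test text.
; No press handling is defined here; overlap, tick and the construction script are empty.
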(event EventBeginPlay
  (|CallParentFunction)
  (Networking|CanExecuteCosmeticEvents
    (:True
      (Rendering|SetVisibility (Variables|Default|GetStaticMesh) false false))))

(event Collision|EventActorBeginOverlap
  (params (OtherActor :type Actor%20Object%20Reference)))

(event EventTick
  (params (DeltaSeconds :type Float%20%28single-precision%29)))

(event InteractInterface|EventOnInteractFocused
  (params (FortPC :type Fort%20Player%20Controller%20Object%20Reference) (ContextInfo :type Fort%20Interact%20Context%20Info%20Object%20Reference))
  (Networking|CanExecuteCosmeticEvents
    (:True
      (Rendering|SetVisibility (Variables|Default|GetStaticMesh) true false))))

(event InteractInterface|EventOnInteractUnfocused
  (params (FortPC :type Fort%20Player%20Controller%20Object%20Reference) (NewContextInfo :type Fort%20Interact%20Context%20Info%20Object%20Reference))
  (Networking|CanExecuteCosmeticEvents
    (:True
      (Rendering|SetVisibility (Variables|Default|GetStaticMesh) false false))))

(fn BlueprintCanInteract
  (params (InteractingPawn :type Fort%20Pawn%20Object%20Reference) (InteractionBeingAttempted :type EInteractionBeingAttempted%20Enum :default "FirstInteraction") (InteractionType :type TInteractionType%20Enum :default "IT_NoInteraction"))
  (return (Variables|Default|GetbCanInteract)))

(fn BlueprintGetInteractionString
  (params (InteractingPawn :type Fort%20Pawn%20Object%20Reference) (InteractionBeingAttempted :type EInteractionBeingAttempted%20Enum :default "FirstInteraction"))
  (return "NSLOCTEXT(\"[2BEFBC1BB6369292E86D2C32C1EA49C8]\", \"AAC4A18A4C32F30B3752F9A027B92FAE\", \"Press Eenie | TEST\")"))

(fn ConstructionScript)
